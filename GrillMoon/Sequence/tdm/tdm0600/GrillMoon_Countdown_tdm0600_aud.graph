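# Audio Level Sequence for the GrillMoon countdown scene tdm0600, running at a 60 fps display rate.
# An event track on the POI music component, bound under the Day 6 sequence runner, fires two triggers whose target functions are not shown.
# A root audio track holds one section at constant volume and pitch, starting after the first trigger and spanning the second.
# The countdown mask timer is bound but has no tracks.
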LevelSequence /GrillMoon/Sequence/tdm/tdm0600/GrillMoon_Countdown_tdm0600_aud
bindings [
 GrillMoon_Countdown_BPmaskTimer {id=#1, possessed_class=GrillMoon_Countdown_BPmaskTimer_C, tracks []}
 BP_GrillMoon_POI_MusicComponent {
  id=#2
  parent_id=#3
  possessed_class=BP_GrillMoon_POI_MusicComponent_C
  tracks [{
    type=MovieSceneEventTrack
    sections [{
      type=MovieSceneEventTriggerSection
      channels [None {
        type=MovieSceneScriptingEventChannel
        keys [[3626.2825, "<Struct 'MovieSceneEvent' (0x85024350780) {}>"], [4045.295, "<Struct 'MovieSceneEvent' (0x85024359e20) {}>"]]
       }]
     }]
   }]
 }
 BP_GrillMoon_Day6_SequenceRunner {id=#3, possessed_class=null, tracks []}
]
display_rate {denominator=1, numerator=60}
tick_resolution {denominator=1, numerator=24000}
tracks [{
  type=MovieSceneAudioTrack
  display_name=Audio
  sections [{
    type=MovieSceneAudioSection
    end_frame=4349
    start_frame=3723
    channels [
     Volume {type=MovieSceneScriptingFloatChannel, default=1}
     Pitch {type=MovieSceneScriptingFloatChannel, default=1}
     AttachActor {type=MovieSceneScriptingActorReferenceChannel}
    ]
   }]
 }]
