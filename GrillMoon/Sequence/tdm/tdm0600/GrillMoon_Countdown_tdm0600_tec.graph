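# Level Sequence for the tdm0600 shot of the GrillMoon countdown, driving the mask timer's effects and material phase.
# It possesses GrillMoon_Countdown_BPmaskTimer and switches seven of its child Niagara components, six Masks systems and PortalContact, on and off together over one shared window.
# Within that window, a track on the MPC_GrillMoon Material Parameter Collection animates Mask_Phase from 5 to 6, changing the material state alongside the particles.
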LevelSequence /GrillMoon/Sequence/tdm/tdm0600/GrillMoon_Countdown_tdm0600_tec
bindings [
 NS_GrillMoon_Countdown_Masks_005 {
  id=#1
  parent_id=#2
  possessed_class=NiagaraComponent
  tracks [{
    type=MovieSceneParticleTrack
    display_name=FX System
    sections [{
      type=MovieSceneParticleSection
      channels [None {type=MovieSceneScriptingParticleChannel, keys [[3683, "<ParticleKey.ACTIVATE: 0>"], [4112, "<ParticleKey.DEACTIVATE: 1>"]]}]
     }]
   }]
 }
 GrillMoon_Countdown_BPmaskTimer {id=#2, possessed_class=GrillMoon_Countdown_BPmaskTimer_C, tracks []}
 NS_GrillMoon_Countdown_Masks_003 {
  id=#3
  parent_id=#2
  possessed_class=NiagaraComponent
  tracks [{
    type=MovieSceneParticleTrack
    display_name=FX System
    sections [{
      type=MovieSceneParticleSection
      channels [None {type=MovieSceneScriptingParticleChannel, keys [[3683, "<ParticleKey.ACTIVATE: 0>"], [4112, "<ParticleKey.DEACTIVATE: 1>"]]}]
     }]
   }]
 }
 NS_GrillMoon_Countdown_Masks_002 {
  id=#4
  parent_id=#2
  possessed_class=NiagaraComponent
  tracks [{
    type=MovieSceneParticleTrack
    display_name=FX System
    sections [{
      type=MovieSceneParticleSection
      channels [None {type=MovieSceneScriptingParticleChannel, keys [[3683, "<ParticleKey.ACTIVATE: 0>"], [4112, "<ParticleKey.DEACTIVATE: 1>"]]}]
     }]
   }]
 }
 NS_GrillMoon_Countdown_PortalContact {
  id=#5
  parent_id=#2
  possessed_class=NiagaraComponent
  tracks [{
    type=MovieSceneParticleTrack
    display_name=FX System
    sections [{
      type=MovieSceneParticleSection
      channels [None {type=MovieSceneScriptingParticleChannel, keys [[3683, "<ParticleKey.ACTIVATE: 0>"], [4112, "<ParticleKey.DEACTIVATE: 1>"]]}]
     }]
   }]
 }
 NS_GrillMoon_Countdown_Masks_001 {
  id=#6
  parent_id=#2
  possessed_class=NiagaraComponent
  tracks [{
    type=MovieSceneParticleTrack
    display_name=FX System
    sections [{
      type=MovieSceneParticleSection
      channels [None {type=MovieSceneScriptingParticleChannel, keys [[3683, "<ParticleKey.ACTIVATE: 0>"], [4112, "<ParticleKey.DEACTIVATE: 1>"]]}]
     }]
   }]
 }
 NS_GrillMoon_Countdown_Masks_004 {
  id=#7
  parent_id=#2
  possessed_class=NiagaraComponent
  tracks [{
    type=MovieSceneParticleTrack
    display_name=FX System
    sections [{
      type=MovieSceneParticleSection
      channels [None {type=MovieSceneScriptingParticleChannel, keys [[3683, "<ParticleKey.ACTIVATE: 0>"], [4112, "<ParticleKey.DEACTIVATE: 1>"]]}]
     }]
   }]
 }
 NS_GrillMoon_Countdown_Masks_006 {
  id=#8
  parent_id=#2
  possessed_class=NiagaraComponent
  tracks [{
    type=MovieSceneParticleTrack
    display_name=FX System
    sections [{
      type=MovieSceneParticleSection
      channels [None {type=MovieSceneScriptingParticleChannel, keys [[3683, "<ParticleKey.ACTIVATE: 0>"], [4112, "<ParticleKey.DEACTIVATE: 1>"]]}]
     }]
   }]
 }
]
display_rate {denominator=1, numerator=60}
tick_resolution {denominator=1, numerator=24000}
tracks [{
  type=MovieSceneMaterialParameterCollectionTrack
  display_name=MPC_GrillMoon
  sections [{type=MovieSceneParameterSection, channels [Mask_Phase {type=MovieSceneScriptingFloatChannel, default=6, keys [[3904,5],[4006,6]]}]}]
 }]
